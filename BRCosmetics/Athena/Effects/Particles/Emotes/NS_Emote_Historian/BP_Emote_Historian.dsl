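; Blueprint for the Historian emote effect that animates a "Time" material parameter on its skeletal mesh.
; BeginPlay waits one second, then plays a timeline whose track drives that scalar from 0 to 1 through a lerp.
; Each update retriggers a short delay, so Time resets to 0 only after updates stop for that interval.
; Tick, overlap and construction script are empty.
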(event Collision|EventActorBeginOverlap
  (params (OtherActor :type Actor%20Object%20Reference)))

(event EventTick
  (params (DeltaSeconds :type Float%20%28single-precision%29)))

(event EventBeginPlay
  (Utilities|FlowControl|RetriggerableDelay 1.0)
  (bind (_direction _time) (Timeline|Timeline_0 PlayFromStart)
    (:Update
      (Rendering|Material|SetScalarParameterValueonMaterials (Variables|Default|GetSkeletalMesh) "Time" (Math|Float|Lerp 0.0 1.0 _time))
      (Utilities|FlowControl|RetriggerableDelay 0.55)
      (Rendering|Material|SetScalarParameterValueonMaterials (Variables|Default|GetSkeletalMesh) "Time" 0.0))))

(fn ConstructionScript)
