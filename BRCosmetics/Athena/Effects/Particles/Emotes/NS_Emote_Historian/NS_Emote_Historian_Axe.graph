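# Niagara effect for the axe in the Historian emote, wrapping it in blue frost, glow, sparks and a timed ice mesh.
# Paired shaft and blade emitters spawn drifting sub-UV frost sprites and soft glows, with glow color picked by time of day.
# Vortex-driven sparks swirl around shaft and blade, while drop sparks fall under gravity and curl noise.
# A single mesh particle animates alpha and a dynamic material parameter alongside a particle light that ramps up.
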NiagaraSystem /BRCosmetics/Athena/Effects/Particles/Emotes/NS_Emote_Historian/NS_Emote_Historian_Axe
dependencies {
 usedDataInterfaces [&NiagaraDataInterfaceCurve]
 usedDynamicInputs []
 usedModules [/Niagara/Modules/System/SystemLifeCycle, &SystemState]
}
emitters {
 Frost {
  emitterUpdateScript [
   &EmitterState(Life Cycle Mode=::NewEnumerator1, Loop Behavior=::NewEnumerator1, Loop Duration=2.1, UseLoopDelay=false)
   &SpawnRate(SpawnRate=&RandomRangeFloat(Minimum=22, Maximum=25), Use Spawn Probability=false)
  ]
  particleSpawnScript [
   /Niagara/Modules/Spawn/Initialization/InitializeParticle(
    Color=<rgba 0.35 0.5992 1 0.07>
    Lifetime=&UniformRangedFloat(Minimum=1.3, Maximum=1.5)
    Module.Write Color=true
    Module.Write Lifetime=true
    Module.Write Mass=false
    Module.Write Material Random=true
    Module.Write Position=false
    Module.Write Scale=false
    Module.Write SpriteRotation=true
    Module.Write SpriteSize=true
    Position=&SimulationPosition()
    Sprite Rotation=&UniformRangedFloat(Minimum=-360, Maximum=360)
    Sprite Size=&Vector2DFromFloat(Value=&UniformRangedFloat(Minimum=11, Maximum=14))
   )
   &SystemLocation(MaskSpawnGroup=false, Offset=&RandomRangeVector(Minimum=<0 0 50>, Maximum=<0 0 -30>))
   &CameraOffset(Camera Offset Amount=&RandomRangeFloat(Minimum=5, Maximum=5))
  ]
  particleUpdateScript [
   &ParticleState()
   &AddVelocity(Coordinate Space=::World, Velocity=&RandomRangeVector(Minimum=<0.1 0.1 -0.3>, Maximum=<-0.1 -0.1 0>))
   &SolveForcesAndVelocity(Clamp Velocity=false, Limit Acceleration=false)
   &ScaleSpriteSize(Scale Factor=&Vector2DFromFloat(Value=&FloatFromCurve(FloatCurve=DataInterface() {propertyValues {Curve {keys [<at 0 0.5>,<at 1 1.3>]}}})))
   &ScaleColor(
    Scale Alpha=&FloatFromCurve(FloatCurve=DataInterface() {propertyValues {Curve {keys [<at 0 0 RCIM_Cubic>,<at 0.4 1 RCIM_Cubic>,<at 1 0 RCIM_Cubic>]}}})
    ScaleA=true
    ScaleRGB=true
   )
   &SubUVAnimation(
    End Frame=26
    SubUV Animation Mode:ENiagara_SubUVLookupMode=::NewEnumerator1
    SubUV Curve=DataInterface() {propertyValues {Curve {keys [
        {arriveTangent=2.3868, interpMode=RCIM_Cubic, leaveTangent=2.3868, tangentMode=RCTM_User, time=0, value=0}
        {interpMode=RCIM_Cubic, tangentMode=RCTM_User, time=1, value=1}
       ]}}}
   )
  ]
  rendererClasses [&NiagaraSpriteRendererProperties]
  renderers [{rendererClass=&NiagaraSpriteRendererProperties}]
 }
 Frost_Blade {
  emitterUpdateScript [
   &EmitterState(Life Cycle Mode=::NewEnumerator1, Loop Behavior=::NewEnumerator1, Loop Duration=2.1, UseLoopDelay=false)
   &SpawnRate(SpawnRate=&RandomRangeFloat(Minimum=30, Maximum=25), Use Spawn Probability=false)
  ]
  particleSpawnScript [
   &InitializeParticle(
    Color=<rgba 0.35 0.5992 1 0.07>
    Lifetime=&UniformRangedFloat(Minimum=1.3, Maximum=1.5)
    Module.Write Color=true
    Module.Write Lifetime=true
    Module.Write Mass=false
    Module.Write Material Random=true
    Module.Write Position=false
    Module.Write Scale=false
    Module.Write SpriteRotation=true
    Module.Write SpriteSize=true
    Position=&SimulationPosition()
    Sprite Rotation=&UniformRangedFloat(Minimum=-360, Maximum=360)
    Sprite Size=&Vector2DFromFloat(Value=&UniformRangedFloat(Minimum=10, Maximum=13))
   )
   &SystemLocation(MaskSpawnGroup=false, Offset=&RandomRangeVector(Minimum=<10 0 50>, Maximum=<-20 0 60>))
   &CameraOffset(Camera Offset Amount=&RandomRangeFloat(Minimum=5, Maximum=5))
  ]
  particleUpdateScript [
   &ParticleState()
   &AddVelocity(Coordinate Space=::World, Velocity=&RandomRangeVector(Minimum=<0.1 0.1 -0.4>, Maximum=<-0.1 -0.1 0>))
   &SolveForcesAndVelocity(Clamp Velocity=false, Limit Acceleration=false)
   &ScaleSpriteSize(Scale Factor=&Vector2DFromFloat(Value=&FloatFromCurve(FloatCurve=DataInterface() {propertyValues {Curve {keys [<at 0 0.5>,<at 1 1.3>]}}})))
   &ScaleColor(
    Scale Alpha=&FloatFromCurve(FloatCurve=DataInterface() {propertyValues {Curve {keys [<at 0 0 RCIM_Cubic>,<at 0.4 1 RCIM_Cubic>,<at 1 0 RCIM_Cubic>]}}})
    ScaleA=true
    ScaleRGB=true
   )
   &SubUVAnimation(
    End Frame=26
    SubUV Animation Mode:ENiagara_SubUVLookupMode=::NewEnumerator1
    SubUV Curve=DataInterface() {propertyValues {Curve {keys [
        {arriveTangent=2.3868, interpMode=RCIM_Cubic, leaveTangent=2.3868, tangentMode=RCTM_User, time=0, value=0}
        {interpMode=RCIM_Cubic, tangentMode=RCTM_User, time=1, value=1}
       ]}}}
   )
  ]
  rendererClasses [&NiagaraSpriteRendererProperties]
  renderers [{rendererClass=&NiagaraSpriteRendererProperties}]
 }
 Glow {
  emitterUpdateScript [
   &EmitterState(Life Cycle Mode=::NewEnumerator1, Loop Behavior=::NewEnumerator1, Loop Duration=2, UseLoopDelay=false)
   &SpawnRate(SpawnRate=&RandomRangeFloat(Minimum=30, Maximum=25), Use Spawn Probability=false)
  ]
  particleSpawnScript [
   &InitializeParticle(
    Lifetime=&UniformRangedFloat(Minimum=1.5, Maximum=2)
    Module.Write Color=true
    Module.Write Lifetime=true
    Module.Write Mass=false
    Module.Write Material Random=true
    Module.Write Position=false
    Module.Write Scale=false
    Module.Write SpriteRotation=true
    Module.Write SpriteSize=true
    Position=&SimulationPosition()
    Sprite Rotation=&UniformRangedFloat(Minimum=-360, Maximum=360)
    Sprite Size=&Vector2DFromFloat(Value=&UniformRangedFloat(Minimum=24, Maximum=30))
   )
   /BRCosmetics/Weapons/FORT_Melee/Pickaxe_Pumpkin_Spice_1h/FX/ColorFromTOD(
    Afternoon:LinearColor=<rgba 0.05 0.335 1 0.05>
    Evening:LinearColor=<rgba 0.35 0.5667 1 0.05>
    FortniteNPC_FortniteActiveTimeOfDay:LinearColor=$NPC.FortniteNPC.FortniteActiveTimeOfDay
    Morning:LinearColor=<rgba 0.35 0.5667 1 0.05>
    Night:LinearColor=<rgba 0.35 0.5667 1 0.05>
   )
   &SystemLocation(MaskSpawnGroup=false, Offset=&RandomRangeVector(Minimum=<0 0 50>, Maximum=<0 0 -35>))
   &CameraOffset(Camera Offset Amount=&RandomRangeFloat(Minimum=5, Maximum=5))
  ]
  particleUpdateScript [
   &ParticleState()
   &AddVelocity(Velocity=&RandomRangeVector(Minimum=<0 0 -0.1>, Maximum=<0 0 0>))
   &SolveForcesAndVelocity(Clamp Velocity=false, Limit Acceleration=false)
   &ScaleSpriteSize(Scale Factor=&Vector2DFromFloat(Value=&FloatFromCurve(FloatCurve=DataInterface() {propertyValues {Curve {keys [<at 0 0.5>,<at 1 1>]}}})))
   &ScaleColor(
    Scale Alpha=&FloatFromCurve(FloatCurve=DataInterface() {propertyValues {Curve {keys [<at 0 0 RCIM_Cubic>,<at 0.4 1 RCIM_Cubic>,<at 1 0 RCIM_Cubic>]}}})
    ScaleA=true
    ScaleRGB=true
   )
  ]
  rendererClasses [&NiagaraSpriteRendererProperties]
  renderers [{rendererClass=&NiagaraSpriteRendererProperties}]
 }
 Glow_Blade {
  emitterUpdateScript [
   &EmitterState(Life Cycle Mode=::NewEnumerator1, Loop Behavior=::NewEnumerator1, Loop Duration=2, UseLoopDelay=false)
   &SpawnRate(SpawnRate=&RandomRangeFloat(Minimum=20, Maximum=25), Use Spawn Probability=false)
  ]
  particleSpawnScript [
   &InitializeParticle(
    Lifetime=&UniformRangedFloat(Minimum=1.5, Maximum=2)
    Module.Write Color=true
    Module.Write Lifetime=true
    Module.Write Mass=false
    Module.Write Material Random=true
    Module.Write Position=false
    Module.Write Scale=false
    Module.Write SpriteRotation=true
    Module.Write SpriteSize=true
    Position=&SimulationPosition()
    Sprite Rotation=&UniformRangedFloat(Minimum=-360, Maximum=360)
    Sprite Size=&Vector2DFromFloat(Value=&UniformRangedFloat(Minimum=24, Maximum=30))
   )
   &ColorFromTOD(
    Afternoon:LinearColor=<rgba 0.05 0.335 1 0.05>
    Evening:LinearColor=<rgba 0.35 0.5667 1 0.05>
    FortniteNPC_FortniteActiveTimeOfDay:LinearColor=$NPC.FortniteNPC.FortniteActiveTimeOfDay
    Morning:LinearColor=<rgba 0.35 0.5667 1 0.05>
    Night:LinearColor=<rgba 0.35 0.5667 1 0.05>
   )
   &SystemLocation(MaskSpawnGroup=false, Offset=&RandomRangeVector(Minimum=<10 0 55>, Maximum=<-20 0 70>))
   &CameraOffset(Camera Offset Amount=&RandomRangeFloat(Minimum=5, Maximum=5))
  ]
  particleUpdateScript [
   &ParticleState()
   &AddVelocity(Velocity=&RandomRangeVector(Minimum=<0 0 -0.1>, Maximum=<0 0 0>))
   &SolveForcesAndVelocity(Clamp Velocity=false, Limit Acceleration=false)
   &ScaleSpriteSize(Scale Factor=&Vector2DFromFloat(Value=&FloatFromCurve(FloatCurve=DataInterface() {propertyValues {Curve {keys [<at 0 0.5>,<at 1 1>]}}})))
   &ScaleColor(
    Scale Alpha=&FloatFromCurve(FloatCurve=DataInterface() {propertyValues {Curve {keys [<at 0 0 RCIM_Cubic>,<at 0.4 1 RCIM_Cubic>,<at 1 0 RCIM_Cubic>]}}})
    ScaleA=true
    ScaleRGB=true
   )
  ]
  rendererClasses [&NiagaraSpriteRendererProperties]
  renderers [{rendererClass=&NiagaraSpriteRendererProperties}]
 }
 Ice_Light {
  emitterUpdateScript [
   &EmitterState(Life Cycle Mode=::NewEnumerator1, Loop Behavior=::NewEnumerator1, Loop Duration=3.5, UseLoopDelay=false)
   &SpawnBurst_Instantaneous(Spawn Count=1, Use Spawn Probability=false)
  ]
  particleSpawnScript [
   &InitializeParticle(
    Color=<rgba 1.5 2.6667 5 1>
    Lifetime=2.35
    Module.Write Color=true
    Module.Write Lifetime=true
    Module.Write Mass=true
    Module.Write Material Random=true
    Module.Write Position=false
    Module.Write Scale=true
    Module.Write SpriteRotation=false
    Module.Write SpriteSize=true
    Position=&SimulationPosition()
    Sprite Size=&Vector2DFromFloat(Value=1)
   )
   &SystemLocation(MaskSpawnGroup=false, Offset=<1.38 0 62.6>)
  ]
  particleUpdateScript [
   &ParticleState()
   &SolveForcesAndVelocity(Clamp Velocity=false, Limit Acceleration=false)
   &ScaleColor(
    Scale RGB=&VectorFromFloat(Value=&FloatFromCurve(FloatCurve=DataInterface() {propertyValues {Curve {keys [<at 0.1 0>,<at 0.6 1>]}}}))
    ScaleA=true
    ScaleRGB=true
   )
   &FN_Light_v2(Light Exponent=2, Light Radius=100, LightEnabled=true)
  ]
  rendererClasses [&NiagaraLightRendererProperties]
  renderers [{rendererClass=&NiagaraLightRendererProperties, rendererIndex=0}]
 }
 Ice_Mesh {
  emitterUpdateScript [
   &EmitterState(Life Cycle Mode=::NewEnumerator1, Loop Behavior=::NewEnumerator1, Loop Duration=3.5, UseLoopDelay=false)
   &SpawnBurst_Instantaneous(Spawn Count=1, Use Spawn Probability=false)
  ]
  particleSpawnScript [
   &InitializeParticle(
    Lifetime=2.55
    Mesh Scale=<1.01 1.2 1.02>
    Module.Write Color=true
    Module.Write Lifetime=true
    Module.Write Mass=true
    Module.Write Material Random=true
    Module.Write Position=false
    Module.Write Scale=true
    Module.Write SpriteRotation=false
    Module.Write SpriteSize=true
    Position=&SimulationPosition()
   )
   &SystemLocation(MaskSpawnGroup=false, Offset=<1.38 0 62.6>)
  ]
  particleUpdateScript [
   &ParticleState()
   &SolveForcesAndVelocity(Clamp Velocity=false, Limit Acceleration=false)
   &ScaleColor(
    Scale Alpha=&FloatFromCurve(FloatCurve=DataInterface() {propertyValues {Curve {keys [
         {interpMode=RCIM_Cubic, tangentMode=RCTM_Break, time=0, value=0}
         {arriveTangent=-0.284, interpMode=RCIM_Cubic, leaveTangent=-0.284, tangentMode=RCTM_User, time=0.6, value=1}
        ]}}})
    ScaleA=true
    ScaleRGB=true
   )
   &DynamicMaterialParameters(
    Index 0 Param 1=&FloatFromCurve(FloatCurve=DataInterface() {propertyValues {Curve {keys [{interpMode=RCIM_Cubic, tangentMode=RCTM_Break, time=0, value=1}, <at 0.6 0>, <at 0.9 0>, <at 1 1>]}}})
    Module.Param0WriteEnabled=true
    Module.Param1WriteEnabled=true
    Module.Param2WriteEnabled=true
    Module.Param3WriteEnabled=true
   )
   &FN_Light_v3(
    Color=<rgba 0.2 0.6 1 1>
    Create Color Attribute=true
    Create Exponent Attribute=true
    Create Radius Attribute=true
    Enable Particle Light Enable=true
    Exponent=2
    Light Enabled=true
    Radius=5
   )
  ]
  rendererClasses [&NiagaraMeshRendererProperties]
  renderers [{rendererClass=&NiagaraMeshRendererProperties, rendererIndex=0}]
 }
 Spark {
  emitterUpdateScript [
   &EmitterState(Life Cycle Mode=::NewEnumerator1, Loop Behavior=::NewEnumerator1, Loop Duration=2, UseLoopDelay=false)
   &SpawnRate(SpawnRate=10, Use Spawn Probability=false)
   &SpawnBurst_Instantaneous(Spawn Count=10, Use Spawn Probability=false)
  ]
  particleSpawnScript [
   &InitializeParticle(
    Color=<rgba 0.5 0.75 1 1>
    Lifetime=&RandomRangeFloat(Minimum=1.8, Maximum=2.2)
    Module.Write Color=true
    Module.Write Lifetime=true
    Module.Write Mass=true
    Module.Write Material Random=true
    Module.Write Position=false
    Module.Write Scale=false
    Module.Write SpriteRotation=false
    Module.Write SpriteSize=true
    Position=&SimulationPosition()
    Sprite Size=&RandomRangeVector2D(Minimum=<0.5 1.5>, Maximum=<1 1>)
   )
   &SystemLocation(MaskSpawnGroup=false, Offset=<0 0 60>)
   &CylinderLocation(
    Cylinder Height=0
    Cylinder Radius=5
    MaskSpawnGroup=false
    Module.Enable Lathe Profile=false
    Module.Override Randomness=false
    Module.Surface Only=false
   )
   &AddVelocity(Velocity=&RandomRangeVector(Minimum=<0 -6 -20>, Maximum=<1 -3 -50>))
  ]
  particleUpdateScript [
   &ParticleState()
   /Niagara/Modules/Update/Velocity/VortexVelocity(
    Module.Use Influence Falloff=false
    Velocity Amount:NiagaraFloat=&RandomRangeFloat(Minimum=40, Maximum=120)
   )
   &SolveForcesAndVelocity(Clamp Velocity=false, Limit Acceleration=false)
   &ScaleSpriteSize(Scale Factor=&Vector2DFromFloat(Value=&FloatFromCurve(FloatCurve=DataInterface() {propertyValues {Curve {keys [<at 0 0>,<at 0.2 1>,<at 0.7 1>,<at 1 0>]}}}, Scale Curve=0.7)))
  ]
  rendererClasses [&NiagaraSpriteRendererProperties]
  renderers [{rendererClass=&NiagaraSpriteRendererProperties}]
 }
 Spark_Blade {
  emitterUpdateScript [
   &EmitterState(Life Cycle Mode=::NewEnumerator1, Loop Behavior=::NewEnumerator1, Loop Duration=1.5, UseLoopDelay=false)
   &SpawnRate(SpawnRate=30, Use Spawn Probability=false)
   &SpawnBurst_Instantaneous(Spawn Count=10, Use Spawn Probability=false)
  ]
  particleSpawnScript [
   &InitializeParticle(
    Color=<rgba 0.5 0.75 1 1>
    Lifetime=&RandomRangeFloat(Minimum=1.8, Maximum=2.2)
    Module.Write Color=true
    Module.Write Lifetime=true
    Module.Write Mass=true
    Module.Write Material Random=true
    Module.Write Position=false
    Module.Write Scale=false
    Module.Write SpriteRotation=false
    Module.Write SpriteSize=true
    Position=&SimulationPosition()
    Sprite Size=&RandomRangeVector2D(Minimum=<0.5 1.5>, Maximum=<1 1>)
   )
   &SystemLocation(MaskSpawnGroup=false, Offset=<15 0 60>)
   &CylinderLocation(
    Cylinder Height=0
    Cylinder Radius=2
    MaskSpawnGroup=false
    Module.Enable Lathe Profile=false
    Module.Override Randomness=false
    Module.Surface Only=false
   )
   &AddVelocity(Velocity=&RandomRangeVector(Minimum=<-20 0 0>, Maximum=<-10 0 0>))
  ]
  particleUpdateScript [
   &VortexVelocity(
    Module.Use Influence Falloff=false
    Velocity Amount:NiagaraFloat=&RandomRangeFloat(Minimum=10, Maximum=15)
    Vortex Axis:Vector3f=<5 0 0>
    Vortex Origin:NiagaraPosition=$Particles.Initial.Position
   )
   &ParticleState()
   &SolveForcesAndVelocity(Clamp Velocity=false, Limit Acceleration=false)
   &ScaleSpriteSize(Scale Factor=&Vector2DFromFloat(Value=&FloatFromCurve(FloatCurve=DataInterface() {propertyValues {Curve {keys [<at 0 0>,<at 0.2 1>,<at 0.7 1>,<at 1 0>]}}}, Scale Curve=0.8)))
  ]
  rendererClasses [&NiagaraSpriteRendererProperties]
  renderers [{rendererClass=&NiagaraSpriteRendererProperties}]
 }
 Spark_Drop {
  emitterUpdateScript [
   &EmitterState(Life Cycle Mode=::NewEnumerator1, Loop Behavior=::NewEnumerator1, Loop Duration=2, UseLoopDelay=false)
   &SpawnRate(SpawnRate=8, Use Spawn Probability=false)
  ]
  particleSpawnScript [
   &InitializeParticle(
    Color=<rgba 0.5 0.75 1 1>
    Lifetime=&RandomRangeFloat(Minimum=1.2, Maximum=0.5)
    Module.Write Color=true
    Module.Write Lifetime=true
    Module.Write Mass=true
    Module.Write Material Random=true
    Module.Write Position=false
    Module.Write Scale=false
    Module.Write SpriteRotation=false
    Module.Write SpriteSize=true
    Position=&SimulationPosition()
    Sprite Size=&RandomRangeVector2D(Minimum=<1.5 1.5>, Maximum=<1 2>)
   )
   &SystemLocation(MaskSpawnGroup=false, Offset=&RandomRangeVector(Minimum=<-20 5 65>, Maximum=<-5 -5 45>))
  ]
  particleUpdateScript [
   &ParticleState()
   &GravityForce(Gravity=<0 0 -100>)
   &CurlNoiseForce(Noise Strength=70, Pan Noise=false)
   &SolveForcesAndVelocity(Clamp Velocity=false, Limit Acceleration=false)
   &ScaleSpriteSize(Scale Factor=&Vector2DFromFloat(Value=&FloatFromCurve(FloatCurve=DataInterface() {propertyValues {Curve {keys [<at 0 0>,<at 0.2 1>,<at 0.7 1>,<at 1 0>]}}}, Scale Curve=0.8)))
  ]
  rendererClasses [&NiagaraSpriteRendererProperties]
  renderers [{rendererClass=&NiagaraSpriteRendererProperties}]
 }
}
summary {
 systemName=NS_Emote_Historian_Axe
 emitters [
  {emitterName=Glow, rendererClasses [&NiagaraSpriteRendererProperties]}
  {emitterName=Frost, rendererClasses [&NiagaraSpriteRendererProperties]}
  {emitterName=Glow_Blade, rendererClasses [&NiagaraSpriteRendererProperties]}
  {emitterName=Frost_Blade, rendererClasses [&NiagaraSpriteRendererProperties]}
  {emitterName=Spark, rendererClasses [&NiagaraSpriteRendererProperties]}
  {emitterName=Spark_Blade, rendererClasses [&NiagaraSpriteRendererProperties]}
  {emitterName=Spark_Drop, rendererClasses [&NiagaraSpriteRendererProperties]}
  {emitterName=Ice_Mesh, rendererClasses [&NiagaraMeshRendererProperties]}
  {emitterName=Ice_Light, rendererClasses [&NiagaraLightRendererProperties]}
 ]
 userVariables [User.SkeletalMesh {
   type=NiagaraDataInterfaceSkeletalMesh
   defaultValue {
    struct=NiagaraExt_VariableValue_DataInterface
    value {dataInterface=&self:NiagaraDataInterfaceSkeletalMesh_0}
   }
  }]
}
