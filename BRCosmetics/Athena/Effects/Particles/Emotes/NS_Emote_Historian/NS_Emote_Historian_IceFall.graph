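# Niagara system for the Historian emote that plays a falling burst of pale-blue frost and ice.
# Three Frost sprite emitters burst flipbook puffs that grow, drift down and fade, with Frost_Long stretching vertically.
# Spinning ice sprites and tumbling mesh shards fall under gravity, while curl noise scatters small falling sparks.
# Emitters spawn in offset boxes around the system location; the exposed SkeletalMesh parameter feeds none of the shown modules.
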NiagaraSystem /BRCosmetics/Athena/Effects/Particles/Emotes/NS_Emote_Historian/NS_Emote_Historian_IceFall
dependencies {
 usedDataInterfaces [&NiagaraDataInterfaceCurve, &NiagaraDataInterfaceVector2DCurve]
 usedDynamicInputs []
 usedModules [/Niagara/Modules/System/SystemLifeCycle, &SystemState]
}
emitters {
 Frost {
  emitterUpdateScript [
   &EmitterState(Life Cycle Mode=::NewEnumerator1, Loop Behavior=::NewEnumerator1, Loop Duration=2.1, UseLoopDelay=false)
   &SpawnBurst_Instantaneous(Spawn Count=20, Use Spawn Probability=false)
  ]
  particleSpawnScript [
   /Niagara/Modules/Spawn/Initialization/InitializeParticle(
    Color=<rgba 0.4 0.63 1 0.07>
    Lifetime=&UniformRangedFloat(Minimum=1.3, Maximum=1.5)
    Module.Write Color=true
    Module.Write Lifetime=true
    Module.Write Mass=false
    Module.Write Material Random=true
    Module.Write Position=false
    Module.Write Scale=false
    Module.Write SpriteRotation=true
    Module.Write SpriteSize=true
    Position=&SimulationPosition()
    Sprite Rotation=&UniformRangedFloat(Minimum=-360, Maximum=360)
    Sprite Size=&Vector2DFromFloat(Value=&UniformRangedFloat(Minimum=11, Maximum=14))
   )
   &SystemLocation(MaskSpawnGroup=false, Offset=&RandomRangeVector(Minimum=<0 0 50>, Maximum=<0 0 -30>))
   &CameraOffset(Camera Offset Amount=&RandomRangeFloat(Minimum=5, Maximum=5))
  ]
  particleUpdateScript [
   &ParticleState()
   &AddVelocity(Coordinate Space=::World, Velocity=&RandomRangeVector(Minimum=<0.1 0.1 -0.8>, Maximum=<-0.1 -0.1 0>))
   &GravityForce(Gravity=<0 0 -10>)
   &SolveForcesAndVelocity(Clamp Velocity=false, Limit Acceleration=false)
   &ScaleSpriteSize(Scale Factor=&Vector2DFromFloat(Value=&FloatFromCurve(FloatCurve=DataInterface() {propertyValues {Curve {keys [<at 0 0.5>,<at 1 1.3>]}}})))
   &ScaleColor(
    Scale Alpha=&FloatFromCurve(FloatCurve=DataInterface() {propertyValues {Curve {keys [<at 0 0 RCIM_Cubic>,<at 0.4 1 RCIM_Cubic>,<at 1 0 RCIM_Cubic>]}}})
    ScaleA=true
    ScaleRGB=true
   )
   &SubUVAnimation(
    End Frame=26
    SubUV Animation Mode:ENiagara_SubUVLookupMode=::NewEnumerator1
    SubUV Curve=DataInterface() {propertyValues {Curve {keys [
        {arriveTangent=2.3868, interpMode=RCIM_Cubic, leaveTangent=2.3868, tangentMode=RCTM_User, time=0, value=0}
        {interpMode=RCIM_Cubic, tangentMode=RCTM_User, time=1, value=1}
       ]}}}
   )
  ]
  rendererClasses [&NiagaraSpriteRendererProperties]
  renderers [{rendererClass=&NiagaraSpriteRendererProperties}]
 }
 Frost_Blade {
  emitterUpdateScript [
   &EmitterState(Life Cycle Mode=::NewEnumerator1, Loop Behavior=::NewEnumerator1, Loop Duration=2.1, UseLoopDelay=false)
   &SpawnBurst_Instantaneous(Spawn Count=22, Use Spawn Probability=false)
  ]
  particleSpawnScript [
   &InitializeParticle(
    Color=<rgba 0.35 0.5992 1 0.12>
    Lifetime=&UniformRangedFloat(Minimum=1.5, Maximum=1)
    Module.Write Color=true
    Module.Write Lifetime=true
    Module.Write Mass=false
    Module.Write Material Random=true
    Module.Write Position=false
    Module.Write Scale=false
    Module.Write SpriteRotation=true
    Module.Write SpriteSize=true
    Position=&SimulationPosition()
    Sprite Rotation=&UniformRangedFloat(Minimum=-360, Maximum=360)
    Sprite Size=&Vector2DFromFloat(Value=&UniformRangedFloat(Minimum=16, Maximum=13))
   )
   &SystemLocation(MaskSpawnGroup=false, Offset=&RandomRangeVector(Minimum=<5 0 50>, Maximum=<-27 0 67>))
   &CameraOffset(Camera Offset Amount=&RandomRangeFloat(Minimum=5, Maximum=5))
  ]
  particleUpdateScript [
   &ParticleState()
   &AddVelocity(Coordinate Space=::World, Velocity=&RandomRangeVector(Minimum=<0.1 0.1 -0.7>, Maximum=<-0.1 -0.1 0>))
   &GravityForce(Gravity=&RandomRangeVector(Minimum=<0 0 -25>, Maximum=<0 0 -15>))
   &SolveForcesAndVelocity(Clamp Velocity=false, Limit Acceleration=false)
   &ScaleSpriteSize(Scale Factor=&Vector2DFromFloat(Value=&FloatFromCurve(FloatCurve=DataInterface() {propertyValues {Curve {keys [{interpMode=RCIM_Cubic, leaveTangent=1.774, tangentMode=RCTM_Break, time=0, value=0.5}, <at 1 1.5>]}}})))
   &ScaleColor(
    Scale Alpha=&FloatFromCurve(FloatCurve=DataInterface() {propertyValues {Curve {keys [
         {arriveTangent=7.456, interpMode=RCIM_Cubic, leaveTangent=7.456, tangentMode=RCTM_User, time=0, value=0}
         <at 0.2 1 RCIM_Cubic>
         <at 1 0 RCIM_Cubic>
        ]}}})
    ScaleA=true
    ScaleRGB=true
   )
   &SubUVAnimation(
    End Frame=26
    SubUV Animation Mode:ENiagara_SubUVLookupMode=::NewEnumerator1
    SubUV Curve=DataInterface() {propertyValues {Curve {keys [
        {arriveTangent=2.3868, interpMode=RCIM_Cubic, leaveTangent=2.3868, tangentMode=RCTM_User, time=0, value=0}
        {interpMode=RCIM_Cubic, tangentMode=RCTM_User, time=1, value=1}
       ]}}}
   )
  ]
  rendererClasses [&NiagaraSpriteRendererProperties]
  renderers [{rendererClass=&NiagaraSpriteRendererProperties}]
 }
 Frost_Long {
  emitterUpdateScript [
   &EmitterState(Life Cycle Mode=::NewEnumerator1, Loop Behavior=::NewEnumerator1, Loop Duration=2.1, UseLoopDelay=false)
   &SpawnBurst_Instantaneous(Spawn Count=3, Use Spawn Probability=false)
   &SpawnBurst_Instantaneous(Spawn Count=2, Spawn Time=0.05, Use Spawn Probability=false) {module=SpawnBurst_Instantaneous001}
  ]
  particleSpawnScript [
   &InitializeParticle(
    Color=<rgba 0.35 0.5992 1 0.15>
    Lifetime=&UniformRangedFloat(Minimum=1.5, Maximum=0.8)
    Module.Write Color=true
    Module.Write Lifetime=true
    Module.Write Mass=false
    Module.Write Material Random=true
    Module.Write Position=false
    Module.Write Scale=false
    Module.Write SpriteRotation=true
    Module.Write SpriteSize=true
    Position=&SimulationPosition()
    Sprite Rotation=&UniformRangedFloat(Minimum=-360, Maximum=360)
    Sprite Size=&RandomRangeVector2D(Minimum=<15 10>, Maximum=<20 25>)
   )
   &SystemLocation(MaskSpawnGroup=false, Offset=&RandomRangeVector(Minimum=<5 0 58>, Maximum=<-2 0 70>))
   &CameraOffset(Camera Offset Amount=&RandomRangeFloat(Minimum=5, Maximum=5))
  ]
  particleUpdateScript [
   &ParticleState()
   &AddVelocity(
    Coordinate Space=::World
    Velocity=&RandomRangeVector(Minimum=<0.1 0.1 -1.8>, Maximum=<-0.1 -0.1 -0.05>)
   )
   &GravityForce(Gravity=&RandomRangeVector(Minimum=<0 0 -2>, Maximum=<0 0 -20>))
   &SolveForcesAndVelocity(Clamp Velocity=false, Limit Acceleration=false)
   &ScaleSpriteSize(Scale Factor=&Vector2DFromCurve(
     Vector2Curve=DataInterface() {propertyValues {
       XCurve {keys [<at 0 1>,<at 1 1.5>]}
       YCurve {keys [
         {interpMode=RCIM_Cubic, leaveTangent=8.7873, tangentMode=RCTM_Break, time=0, value=1}
         {interpMode=RCIM_Cubic, tangentMode=RCTM_Break, time=1, value=4}
        ]}
      }}
     ScaleCurve=<1 1>
    ))
   &ScaleColor(
    Scale Alpha=&FloatFromCurve(FloatCurve=DataInterface() {propertyValues {Curve {keys [
         {arriveTangent=6.3911, interpMode=RCIM_Cubic, leaveTangent=6.3911, tangentMode=RCTM_User, time=0, value=0}
         <at 0.2 1 RCIM_Cubic>
         <at 1 0 RCIM_Cubic>
        ]}}})
    ScaleA=true
    ScaleRGB=true
   )
   &SubUVAnimation(
    End Frame=26
    SubUV Animation Mode:ENiagara_SubUVLookupMode=::NewEnumerator1
    SubUV Curve=DataInterface() {propertyValues {Curve {keys [
        {arriveTangent=2.3868, interpMode=RCIM_Cubic, leaveTangent=2.3868, tangentMode=RCTM_User, time=0, value=0}
        {interpMode=RCIM_Cubic, tangentMode=RCTM_User, time=1, value=1}
       ]}}}
   )
  ]
  rendererClasses [&NiagaraSpriteRendererProperties]
  renderers [{rendererClass=&NiagaraSpriteRendererProperties}]
 }
 Ice {
  emitterUpdateScript [
   &EmitterState(Life Cycle Mode=::NewEnumerator1, Loop Behavior=::NewEnumerator1, Loop Duration=2.1, UseLoopDelay=false)
   &SpawnBurst_Instantaneous(Spawn Count=22, Use Spawn Probability=false)
   &SpawnBurst_Instantaneous(Spawn Count=7, Spawn Time=0.05, Use Spawn Probability=false) {module=SpawnBurst_Instantaneous001}
  ]
  particleSpawnScript [
   &InitializeParticle(
    Color=&RandomRangeLinearColor(Minimum=<rgba 0.25 0.5375 1 1>, Maximum=<rgba 0.45 0.8525 1.5 1>)
    Lifetime=&UniformRangedFloat(Minimum=0.5, Maximum=1.2)
    Module.Write Color=true
    Module.Write Lifetime=true
    Module.Write Mass=false
    Module.Write Material Random=true
    Module.Write Position=false
    Module.Write Scale=false
    Module.Write SpriteRotation=true
    Module.Write SpriteSize=true
    Position=&SimulationPosition()
    Sprite Rotation=&UniformRangedFloat(Minimum=-360, Maximum=360)
    Sprite Size=&Multiply_Vector2(
     A=&RandomRangeVector2D(Minimum=<3 5>, Maximum=<1 3>)
     B=&Vector2DFromFloat(Value=&RandomRangeFloat(Minimum=0.2, Maximum=1.1))
    )
   )
   &SystemLocation(MaskSpawnGroup=false, Offset=&RandomRangeVector(Minimum=<5 5 40>, Maximum=<-25 -5 65>))
   &CameraOffset(Camera Offset Amount=&RandomRangeFloat(Minimum=5, Maximum=5))
  ]
  particleUpdateScript [
   &ParticleState()
   &AddVelocity(Coordinate Space=::World, Velocity=&RandomRangeVector(Minimum=<0.1 0.1 -6>, Maximum=<-0.1 -0.1 -1.5>))
   &GravityForce(Gravity=&RandomRangeVector(Minimum=<0 0 -20>, Maximum=<0 0 -50>))
   &SpriteRotationRate(Rotation Rate=&RandomRangeFloat(Minimum=-400, Maximum=600))
   &SolveForcesAndVelocity(Clamp Velocity=false, Limit Acceleration=false)
   &ScaleSpriteSize(Scale Factor=&Vector2DFromFloat(Value=&FloatFromCurve(FloatCurve=DataInterface() {propertyValues {Curve {keys [<at 0.65 1>,<at 1 0>]}}})))
   &ScaleColor(
    Scale Alpha=&FloatFromCurve(FloatCurve=DataInterface() {propertyValues {Curve {keys [<at 0.8 1 RCIM_Cubic>,<at 1 0 RCIM_Cubic>]}}})
    ScaleA=true
    ScaleRGB=true
   )
  ]
  rendererClasses [&NiagaraSpriteRendererProperties]
  renderers [{rendererClass=&NiagaraSpriteRendererProperties}]
 }
 Ice001 {
  emitterUpdateScript [
   &EmitterState(Life Cycle Mode=::NewEnumerator1, Loop Behavior=::NewEnumerator1, Loop Duration=2.1, UseLoopDelay=false)
   &SpawnBurst_Instantaneous(Spawn Count=4, Use Spawn Probability=false)
   &SpawnBurst_Instantaneous(Spawn Count=2, Spawn Time=0.05, Use Spawn Probability=false) {module=SpawnBurst_Instantaneous001}
  ]
  particleSpawnScript [
   &InitializeParticle(
    Lifetime=&UniformRangedFloat(Minimum=0.5, Maximum=1)
    Mesh Scale=&Multiply_Vector(
     A=<1.7 0.6 0.4>
     B=&VectorFromFloat(Value=&RandomRangeFloat(Minimum=0.15, Maximum=0.05))
    )
    Module.Write Color=true
    Module.Write Lifetime=true
    Module.Write Mass=false
    Module.Write Material Random=true
    Module.Write Position=false
    Module.Write Scale=true
    Module.Write SpriteRotation=true
    Module.Write SpriteSize=true
    Position=&SimulationPosition()
    Sprite Rotation=&UniformRangedFloat(Minimum=-360, Maximum=360)
    Sprite Size=&Multiply_Vector2(
     A=&RandomRangeVector2D(Minimum=<3 5>, Maximum=<1 3>)
     B=&Vector2DFromFloat(Value=&RandomRangeFloat(Minimum=0.2, Maximum=1.3))
    )
   )
   &SystemLocation(MaskSpawnGroup=false, Offset=&RandomRangeVector(Minimum=<5 5 40>, Maximum=<-25 -5 65>))
   &CameraOffset(Camera Offset Amount=&RandomRangeFloat(Minimum=5, Maximum=5))
   /Niagara/Modules/Spawn/Orientation/V2/InitialMeshOrientation(
    Rotation=&RandomRangeVector(Minimum=<-1 -1 -41>, Maximum=<1 1 1>)
    Use Orientation Vector=true
    Use Rotation Vector=true
   )
  ]
  particleUpdateScript [
   &ParticleState()
   &AddVelocity(Coordinate Space=::World, Velocity=&RandomRangeVector(Minimum=<0.1 0.1 -8>, Maximum=<-0.1 -0.1 -1>))
   /Niagara/Modules/Update/Orientation/V2/UpdateMeshOrientation(Rotation Rate=0.5, Rotation Vector:Vector3f=&RandomRangeVector(Minimum=<-1 -1 -1>, Maximum=<1 1 1>))
   &GravityForce(Gravity=&RandomRangeVector(Minimum=<0 0 -80>, Maximum=<0 0 -100>))
   &SolveForcesAndVelocity(Clamp Velocity=false, Limit Acceleration=false)
   &SolveRotationalForcesAndVelocity()
   &ScaleSpriteSize(Scale Factor=&Vector2DFromFloat(Value=&FloatFromCurve(FloatCurve=DataInterface() {propertyValues {Curve {keys [<at 0.65 1>,<at 1 0>]}}})))
   &ScaleColor(
    Scale Alpha=&FloatFromCurve(FloatCurve=DataInterface() {propertyValues {Curve {keys [<at 0.8 1 RCIM_Cubic>,<at 1 0 RCIM_Cubic>]}}})
    ScaleA=true
    ScaleRGB=true
   )
  ]
  rendererClasses [&NiagaraMeshRendererProperties]
  renderers [{rendererClass=&NiagaraMeshRendererProperties, rendererIndex=0}]
 }
 Spark_Drop {
  emitterUpdateScript [
   &EmitterState(Life Cycle Mode=::NewEnumerator1, Loop Behavior=::NewEnumerator1, Loop Duration=0.5, UseLoopDelay=false)
   &SpawnRate(SpawnRate=15, Use Spawn Probability=false)
   &SpawnBurst_Instantaneous(Spawn Count=18, Use Spawn Probability=false)
  ]
  particleSpawnScript [
   &InitializeParticle(
    Color=<rgba 0.5 0.75 1 1>
    Lifetime=&RandomRangeFloat(Minimum=1.8, Maximum=0.5)
    Module.Write Color=true
    Module.Write Lifetime=true
    Module.Write Mass=true
    Module.Write Material Random=true
    Module.Write Position=false
    Module.Write Scale=false
    Module.Write SpriteRotation=false
    Module.Write SpriteSize=true
    Position=&SimulationPosition()
    Sprite Size=&RandomRangeVector2D(Minimum=<1.5 1.5>, Maximum=<0.5 2>)
   )
   &SystemLocation(MaskSpawnGroup=false, Offset=&RandomRangeVector(Minimum=<-20 5 65>, Maximum=<-5 -5 45>))
  ]
  particleUpdateScript [
   &ParticleState()
   &GravityForce(Gravity=<0 0 -100>)
   &CurlNoiseForce(Noise Strength=70, Pan Noise=false)
   &SolveForcesAndVelocity(Clamp Velocity=false, Limit Acceleration=false)
   &ScaleSpriteSize(Scale Factor=&Vector2DFromFloat(Value=&FloatFromCurve(FloatCurve=DataInterface() {propertyValues {Curve {keys [<at 0 0>,<at 0.2 1>,<at 0.7 1>,<at 1 0>]}}}, Scale Curve=0.8)))
  ]
  rendererClasses [&NiagaraSpriteRendererProperties]
  renderers [{rendererClass=&NiagaraSpriteRendererProperties}]
 }
}
summary {
 systemName=NS_Emote_Historian_IceFall
 emitters [
  {emitterName=Frost, rendererClasses [&NiagaraSpriteRendererProperties]}
  {emitterName=Frost_Blade, rendererClasses [&NiagaraSpriteRendererProperties]}
  {emitterName=Frost_Long, rendererClasses [&NiagaraSpriteRendererProperties]}
  {emitterName=Ice, rendererClasses [&NiagaraSpriteRendererProperties]}
  {emitterName=Spark_Drop, rendererClasses [&NiagaraSpriteRendererProperties]}
  {emitterName=Ice001, rendererClasses [&NiagaraMeshRendererProperties]}
 ]
 userVariables [User.SkeletalMesh {
   type=NiagaraDataInterfaceSkeletalMesh
   defaultValue {
    struct=NiagaraExt_VariableValue_DataInterface
    value {dataInterface=&self:NiagaraDataInterfaceSkeletalMesh_0}
   }
  }]
}
